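# First-person fire sound cue for the RedDot assault rifle.
# The main layer mixes one of six random close shots with a quieter mechanical wave, its volume lowered as the TimeSinceStarted parameter rises.
# A parallel low-ammo layer plays a pitched-down SMG warning wave only when LowAmmo is true.
# Each layer has its own sound class override, and concurrency limits stop the oldest instance.
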SoundCue /FlipperGameplay/Items/Weapons/RedDotAR/Audio/Core_AR_RedDot_Fire_1P_Cue
root=Mixer_2
nodes [
 {id=Branch_3, children [Modulator_8,null,null], properties {boolParameterName=LowAmmo}}
 {id=Mixer_0, children [Random_1, WavePlayer_12], properties {inputVolume [1,0.2]}}
 {id=Mixer_2, children [SoundClass_0,SoundClass_1], properties {inputVolume [1.25,1]}}
 {
  id=ModulatorContinuous_0
  children [Modulator_5]
  properties {volumeModulationParams {default=0, maxOutput=0.65, minOutput=1, parameterName=TimeSinceStarted}}
 }
 {id=Modulator_5, children [Mixer_0], properties {pitchMax=1, pitchMin=1}}
 {id=Modulator_8, children [WavePlayer_0], properties {pitchMax=0.9, pitchMin=0.9, volumeMax=1, volumeMin=1}}
 {
  id=Random_1
  children [WavePlayer_6, WavePlayer_7, WavePlayer_8, WavePlayer_9, WavePlayer_10, WavePlayer_11]
  properties {bShouldExcludeFromBranchCulling=true, weights [1,1,1,1,1,1]}
 }
 {
  id=SoundClass_0
  children [Branch_3]
  properties {soundClassOverride=/Game/Sounds/FortSoundClassesAndModes/Fort_WeaponRanged_LowAmmo_Critical_PSM}
 }
 {
  id=SoundClass_1
  children [ModulatorContinuous_0]
  properties {soundClassOverride=&Fort_WeaponRanged_Fire_1P_PSM}
 }
 {id=WavePlayer_0, properties {soundWaveAssetPtr=/FlipperGameplay/Items/Weapons/CoreSMG/Audio/Core_SMG_LowAmmo_02}}
 {id=WavePlayer_10, properties {soundWaveAssetPtr=&/Core_AR_RedDot_FireB_Close_05}}
 {id=WavePlayer_11, properties {soundWaveAssetPtr=&/Core_AR_RedDot_FireB_Close_06}}
 {id=WavePlayer_12, properties {soundWaveAssetPtr=&/Core_AR_RedDot_FireB_Mech_01}}
 {id=WavePlayer_6, properties {soundWaveAssetPtr=&/Core_AR_RedDot_FireB_Close_04}}
 {id=WavePlayer_7, properties {soundWaveAssetPtr=&/Core_AR_RedDot_FireB_Close_01}}
 {id=WavePlayer_8, properties {soundWaveAssetPtr=&/Core_AR_RedDot_FireB_Close_02}}
 {id=WavePlayer_9, properties {soundWaveAssetPtr=&/Core_AR_RedDot_FireB_Close_03}}
]
properties {
 attenuationSettings=&Fort_Weapon_Projectile_FirstPerson_Athena_Att
 bDebug=true
 maxDistance=2097152
 priority=5
 soundClassObject=&Fort_WeaponRanged_Fire_1P_PSM
 volumeMultiplier=0.7
 attenuationOverrides {manualReverbSendLevel=-1.0165, reverbSendMethod=Manual}
 concurrencyOverrides {bLimitToOwner=true, maxCount=2, resolutionRule=StopOldest}
 concurrencySet [&Weapon_Gunfire_FirstPerson_Conc]
}
