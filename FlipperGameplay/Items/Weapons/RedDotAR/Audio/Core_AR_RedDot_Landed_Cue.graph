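# Sound cue for the drop impact when the red-dot AR weapon lands on the ground as an item.
# A Random node picks one of two drop impact waves with equal weight and feeds a Modulator that varies pitch slightly.
# The cue lowers pitch and volume, uses the loot-collect sound class, and overrides attenuation for short-range falloff.
# A concurrency override caps simultaneous instances and blocks new ones once the limit is reached.
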SoundCue /FlipperGameplay/Items/Weapons/RedDotAR/Audio/Core_AR_RedDot_Landed_Cue
root=Modulator_0
nodes [
 {id=Modulator_0, children [Random_1], properties {pitchMax=1.1, pitchMin=0.9}}
 {id=Random_1, children [WavePlayer_0,WavePlayer_4], properties {weights [1,1]}}
 {id=WavePlayer_0, properties {soundWaveAssetPtr=&/Core_AR_RedDot_DropImpact_01}}
 {id=WavePlayer_4, properties {soundWaveAssetPtr=&/Core_AR_RedDot_DropImpact_02}}
]
properties {
 attenuationSettings=&Foley_Attenuation_Basic_Att
 bOverrideAttenuation=true
 bOverrideConcurrency=true
 maxDistance=1000
 pitchMultiplier=0.8
 soundClassObject=&Fort_Loot_Collect
 volumeMultiplier=0.8
 attenuationOverrides {falloffDistance=1000, attenuationShapeExtents {x=0}}
 concurrencyOverrides {maxCount=3, resolutionRule=PreventNew}
}
