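# Sound cue for the out-of-ammo sound of the Red Dot AR weapon.
# A Random node picks among four equally weighted shared rifle out-of-ammo waves and feeds a Modulator with pitch variation disabled.
# The cue plays at reduced volume through a first-person ranged-fire sound class, with close weapon attenuation and a first-person gunfire concurrency set.
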SoundCue /FlipperGameplay/Items/Weapons/RedDotAR/Audio/Core_AR_RedDot_OutofAmmo_Cue
root=Modulator_1
nodes [
 {id=Modulator_1, children [Random_0], properties {pitchMax=1, pitchMin=1}}
 {id=Random_0, children [WavePlayer_0,WavePlayer_1,WavePlayer_2,WavePlayer_3], properties {weights [1,1,1,1]}}
 {id=WavePlayer_0, properties {soundWaveAssetPtr=/Game/Sounds/Fort_Weapons/OutofAmmo/Rifle_OutofAmmo_01}}
 {id=WavePlayer_1, properties {soundWaveAssetPtr=/Game/Sounds/Fort_Weapons/OutofAmmo/Rifle_OutofAmmo_02}}
 {id=WavePlayer_2, properties {soundWaveAssetPtr=/Game/Sounds/Fort_Weapons/OutofAmmo/Rifle_OutofAmmo_03}}
 {id=WavePlayer_3, properties {soundWaveAssetPtr=/Game/Sounds/Fort_Weapons/OutofAmmo/Rifle_OutofAmmo_04}}
]
properties {
 attenuationSettings=&Fort_Weapon_Projectile_Close_Athena_Att
 maxDistance=4950
 soundClassObject=&Fort_WeaponRanged_Fire_1P_PSM
 volumeMultiplier=0.5
 attenuationOverrides {falloffDistance=1600}
 concurrencyOverrides {maxCount=2, resolutionRule=StopOldest}
 concurrencySet [&Weapon_Gunfire_FirstPerson_Conc]
}
