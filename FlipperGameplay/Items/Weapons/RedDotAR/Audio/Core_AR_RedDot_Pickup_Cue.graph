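# Sound cue for picking up the Red Dot assault rifle, routed to the loot collect sound class with weapon pickup attenuation.
# A short fixed delay leads into a modulator, then a random node picks one of three equally weighted pickup waves.
# It plays at reduced volume and references a shared pickup concurrency set, alongside override values for a single instance that blocks new plays.
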SoundCue /FlipperGameplay/Items/Weapons/RedDotAR/Audio/Core_AR_RedDot_Pickup_Cue
root=Delay_0
nodes [
 {id=Delay_0, children [Modulator_0], properties {delayMax=0.05, delayMin=0.05}}
 {id=Modulator_0, children [Random_0]}
 {id=Random_0, children [WavePlayer_0,WavePlayer_2,WavePlayer_3], properties {weights [1,1,1]}}
 {id=WavePlayer_0, properties {soundWaveAssetPtr=&/Core_AR_RedDot_Pickup_01}}
 {id=WavePlayer_2, properties {soundWaveAssetPtr=&/Core_AR_RedDot_Pickup_02}}
 {id=WavePlayer_3, properties {soundWaveAssetPtr=&/Core_AR_RedDot_Pickup_03}}
]
properties {
 attenuationSettings=&Fort_Weapon_Pickup_Att
 maxDistance=3500
 soundClassObject=&Fort_Loot_Collect
 volumeMultiplier=0.3923
 concurrencyOverrides {maxCount=1, resolutionRule=PreventNew}
 concurrencySet [&PickupMaximum_Oldest]
}
