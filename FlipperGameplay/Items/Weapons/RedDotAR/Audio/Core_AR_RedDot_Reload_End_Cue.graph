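# Sound cue for the end-of-reload foley of the Red Dot assault rifle in FlipperGameplay.
# A single wave, Reload_End_02, feeds a one-entry Random node, then a Modulator at default settings, then a short fixed Delay at the root.
# It plays in the weapon foley sound class with weapon reload attenuation and reduced volume.
# The cue overrides concurrency and has its debug flag enabled.
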SoundCue /FlipperGameplay/Items/Weapons/RedDotAR/Audio/Core_AR_RedDot_Reload_End_Cue
root=Delay_0
nodes [
 {id=Delay_0, children [Modulator_0], properties {delayMax=0.05, delayMin=0.05}}
 {id=Modulator_0, children [Random_0]}
 {id=Random_0, children [WavePlayer_1], properties {weights [1]}}
 {id=WavePlayer_1, properties {soundWaveAssetPtr=&/Core_AR_RedDot_Reload_End_02}}
]
properties {
 attenuationSettings=&Fort_Weapon_Reload_Att
 bDebug=true
 bOverrideConcurrency=true
 maxDistance=2250
 soundClassObject=&Fort_WeaponFoley
 volumeMultiplier=0.451
}
