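# Sound cue for the start of the RedDot AR reload, layering reload start, magazine eject and twirl foley.
# A mixer blends three randomized layers: a modulated start variant, an eject variant, and a twirl variant that plays after a short fixed delay.
# Eject and twirl sit slightly lower in the mix.
# It uses weapon reload attenuation and the weapon foley sound class, overrides concurrency, and has debug enabled.
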SoundCue /FlipperGameplay/Items/Weapons/RedDotAR/Audio/Core_AR_RedDot_Reload_Start_Cue
root=Mixer_0
nodes [
 {id=Delay_0, children [Random_2], properties {delayMax=0.2, delayMin=0.2}}
 {id=Mixer_0, children [Modulator_0,Random_1,Delay_0], properties {inputVolume [1,0.9,0.7]}}
 {id=Modulator_0, children [Random_0]}
 {id=Random_0, children [WavePlayer_0,WavePlayer_1,WavePlayer_2], properties {weights [1,1,1]}}
 {id=Random_1, children [WavePlayer_5,WavePlayer_6], properties {weights [1,1]}}
 {id=Random_2, children [WavePlayer_3,WavePlayer_4], properties {weights [1,1]}}
 {id=WavePlayer_0, properties {soundWaveAssetPtr=&/Core_AR_RedDot_Reload_Start_01}}
 {id=WavePlayer_1, properties {soundWaveAssetPtr=&/Core_AR_RedDot_Reload_Start_02}}
 {id=WavePlayer_2, properties {soundWaveAssetPtr=&/Core_AR_RedDot_Reload_Start_03}}
 {id=WavePlayer_3, properties {soundWaveAssetPtr=&/Core_AR_RedDot_ReloadTwirl_01}}
 {id=WavePlayer_4, properties {soundWaveAssetPtr=&/Core_AR_RedDot_ReloadTwirl_02}}
 {id=WavePlayer_5, properties {soundWaveAssetPtr=&/Core_AR_RedDot_ReloadEject_01}}
 {id=WavePlayer_6, properties {soundWaveAssetPtr=&/Core_AR_RedDot_ReloadEject_02}}
]
properties {
 attenuationSettings=&Fort_Weapon_Reload_Att
 bDebug=true
 bOverrideConcurrency=true
 maxDistance=2250
 soundClassObject=&Fort_WeaponFoley
 volumeMultiplier=0.401
}
