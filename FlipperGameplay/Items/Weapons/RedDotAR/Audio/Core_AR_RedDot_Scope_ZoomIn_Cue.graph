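# Sound cue for zooming in with the RedDot AR scope, routed through the weapon foley sound class.
# A mixer layers the scope zoom-in wave at full input with a quieter, random pick between two player run foley waves.
# Separate modulators randomize each layer's pitch slightly upward, with the foley layer shifted higher.
# Overall volume is low, with overridden foley attenuation and a concurrency rule that blocks new instances.
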SoundCue /FlipperGameplay/Items/Weapons/RedDotAR/Audio/Core_AR_RedDot_Scope_ZoomIn_Cue
root=Mixer_1
nodes [
 {id=Mixer_1, children [Modulator_3,Modulator_4], properties {inputVolume [0.4,1]}}
 {id=Modulator_3, children [Random_2], properties {pitchMax=1.2, pitchMin=1.1}}
 {id=Modulator_4, children [WavePlayer_0], properties {pitchMax=1.1, pitchMin=1}}
 {id=Random_2, children [WavePlayer_7,WavePlayer_6], properties {weights [1,1]}}
 {id=WavePlayer_0, properties {soundWaveAssetPtr=&/Core_AR_RedDot_Scope_ZoomIn_01}}
 {id=WavePlayer_6, properties {soundWaveAssetPtr=/Game/Sounds/Fort_Foley_Sounds/Moves/player_run_foley_05}}
 {id=WavePlayer_7, properties {soundWaveAssetPtr=/Game/Sounds/Fort_Foley_Sounds/Moves/player_run_foley_01}}
]
properties {
 attenuationSettings=&Foley_Attenuation_Basic_Att
 bOverrideAttenuation=true
 bOverrideConcurrency=true
 maxDistance=2700
 soundClassObject=&Fort_WeaponFoley
 volumeMultiplier=0.175
 attenuationOverrides {falloffDistance=1500, nonSpatializedRadiusStart=200, attenuationShapeExtents {x=1200}}
 concurrencyOverrides {resolutionRule=PreventNew}
}
