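# Sound cue for the Red Dot AR scope zoom-out, played as quiet weapon foley.
# A mixer layers a dedicated zoom-out wave, passed through a modulator, over a quieter, pitched-up random pick from four reused player-run foley clips.
# The cue overrides attenuation with its own falloff despite referencing a shared foley attenuation asset, and its concurrency override prevents new instances once the limit is reached.
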SoundCue /FlipperGameplay/Items/Weapons/RedDotAR/Audio/Core_AR_RedDot_Scope_ZoomOut_Cue
root=Mixer_1
nodes [
 {id=Mixer_1, children [Modulator_1,Modulator_3], properties {inputVolume [0.3,1.2]}}
 {id=Modulator_1, children [Random_4], properties {pitchMax=1.2, pitchMin=1.1}}
 {id=Modulator_3, children [WavePlayer_0]}
 {id=Random_4, children [WavePlayer_19, WavePlayer_1565, WavePlayer_1563, WavePlayer_1562], properties {weights [1,1,1,1]}}
 {id=WavePlayer_0, properties {soundWaveAssetPtr=&/Core_AR_RedDot_Scope_ZoomOut_01}}
 {id=WavePlayer_1562, properties {soundWaveAssetPtr=/Game/Sounds/Fort_Foley_Sounds/Moves/player_run_foley_12}}
 {id=WavePlayer_1563, properties {soundWaveAssetPtr=/Game/Sounds/Fort_Foley_Sounds/Moves/player_run_foley_10}}
 {id=WavePlayer_1565, properties {soundWaveAssetPtr=/Game/Sounds/Fort_Foley_Sounds/Moves/player_run_foley_03}}
 {id=WavePlayer_19, properties {soundWaveAssetPtr=/Game/Sounds/Fort_Foley_Sounds/Moves/player_run_foley_09}}
]
properties {
 attenuationSettings=&Foley_Attenuation_Basic_Att
 bOverrideAttenuation=true
 bOverrideConcurrency=true
 maxDistance=2700
 soundClassObject=&Fort_WeaponFoley
 volumeMultiplier=0.125
 attenuationOverrides {falloffDistance=1500, nonSpatializedRadiusStart=200, attenuationShapeExtents {x=1200}}
 concurrencyOverrides {resolutionRule=PreventNew}
}
